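# Skeletal mesh material for the Fluffy cosmetic vehicle wheel, with animated blinking eyes and a custom tint.
# It overrides Base Color and Opacity Mask from MF_WheelParent_Simple, then passes the attributes through vehicle damage and de-rez transition functions.
# A time-driven sine sweeps a UV threshold inside the mask's red channel for the blink, while green marks the tinted main area.
# A scalar blends opacity between the mask's blue channel and a separate LOD mask texture.
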Material /VehicleCosmetics/Wheels/Fluffy/Materials/M_Wheel_Fluffy
graph {
 properties {
  bAutomaticallySetUsageInEditor=false
  bUseMaterialAttributes=true
  bUsedWithMeshDeformer=true
  bUsedWithSkeletalMesh=true
  pixelDepthOffsetMode=PDOM_Legacy
  previewMesh=/VehicleCosmetics/Wheels/Fluffy/Meshes/SK_Wheel_Fluffy
  usedShadingModels=Default Lit
 }
 nodes [
  n1 = GetMaterialAttributes(None=MaterialFunctionCall() {materialFunction=&MF_WheelParent_Simple}.MaterialAttributes) {attributeGetTypes [#1]}
  n2 = TextureCoordinate()
  n3 = Multiply(A=Time(), B=ScalarParameter() {defaultValue=0.3, group=5_Special Effects, parameterName=Blink Speed})
  n4 = Subtract(
   A=OneMinus(None=ComponentMask(None=@n2) {g=true})
   B=LinearInterpolate(
    A=ScalarParameter() {defaultValue=2, group=5_Special Effects, parameterName=Blink Level High}
    B=ScalarParameter() {defaultValue=0.6, group=5_Special Effects, parameterName=Blink Level Low}
    Alpha=Sine(None=@n3)
   )
  )
  n5 = TextureSampleParameter2D() {
   group=1_Input Textures
   parameterName=Custom Masks
   samplerType=SAMPLERTYPE_LinearColor
   texture=/VehicleCosmetics/Wheels/Fluffy/Textures/T_Fluffy_Rim_M
  }
  n6 = Multiply(A=Ceil(None=Clamp(None=@n4)), B=@n5.R)
  n7 = VectorParameter() {group=5_Special Effects, parameterName=Custom Color, defaultValue=<rgba 1 0 0.4564 1>}
  n8 = LinearInterpolate(
   A=LinearInterpolate(A=@n1.Base Color, B=Multiply(A=@n6, B=@n7.RGB) {desc=Eyes}, Alpha=@n6)
   B=Multiply(A=@n7.RGB, B=@n5.G) {desc=Main color}
   Alpha=@n5.G
  )
  n9 = LinearInterpolate(
   A=@n5.B
   B=TextureSample() {samplerType=SAMPLERTYPE_LinearColor, texture=/VehicleCosmetics/Wheels/Fluffy/Textures/T_Fluffy_Rim_M2_LOD}.R
   Alpha=ScalarParameter() {desc="1=lod2m_0=lod0m", group=5_Special Effects, parameterName=LODs_Mask}
  )
  MaterialFunctionCall(In=MaterialFunctionCall(Material=SetMaterialAttributes!(MaterialAttributes=@n1.MaterialAttributes, Base Color=@n8, Opacity Mask=@n9) {attributeSetTypes [#1, #2]}) {materialFunction=&MF_BaseVehicle_Damage}.Result) {materialFunction=&MF_VehicleDeRez_Transition}
  Reroute(None=@n5.B)
 ]
}
