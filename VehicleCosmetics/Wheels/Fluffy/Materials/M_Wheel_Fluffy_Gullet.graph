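# Fluffy wheel "Gullet" vehicle cosmetic material that fakes depth with layered parallax.
# Two bump-offset layers of albedo and normal blend by mask green, then give way to an undisplaced layer by mask red, with Custom Color tinting the albedo.
# A scalar parameter sets roughness; the mask-based roughness blend does not reach the output.
# Attributes pass through vehicle damage and de-rez transition functions; usage covers skeletal meshes and mesh deformers.
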Material /VehicleCosmetics/Wheels/Fluffy/Materials/M_Wheel_Fluffy_Gullet
graph {
 properties {
  bAutomaticallySetUsageInEditor=false
  bUseMaterialAttributes=true
  bUsedWithMeshDeformer=true
  bUsedWithSkeletalMesh=true
  usedShadingModels=Default Lit
 }
 nodes [
  n1 = NamedRerouteUsage()
  n2 = NamedRerouteUsage()
  n3 = ScalarParameter() {parameterName=Roughness}
  MaterialFunctionCall(In=MaterialFunctionCall(Material=SetMaterialAttributes!(MaterialAttributes, Base Color=@n1, Normal=@n2, Roughness=@n3) {attributeSetTypes [#1, #2, #3]}) {materialFunction=&MF_BaseVehicle_Damage}.Result) {materialFunction=&MF_VehicleDeRez_Transition}
  n4 = VectorParameter() {group=00 Base, parameterName=Custom Color, sortPriority=-1, defaultValue=<rgba 0.624 0.0284 0.0762 1>}
  n5 = TextureCoordinate()
  n6 = BumpOffset(Coordinate=@n5, Height=Constant() {r=-9})
  n7 = BumpOffset(Coordinate=@n5, Height=Constant() {r=-2})
  n8 = TextureSample(UVs=@n7, Tex=NamedRerouteUsage()) {samplerType=SAMPLERTYPE_LinearColor, texture=/VehicleCosmetics/Wheels/Fluffy/Textures/T_Fluffy_Gullet_M}
  n9 = LinearInterpolate(
   A=TextureSample(UVs=@n6, Tex=NamedRerouteUsage()) {texture=/VehicleCosmetics/Wheels/Fluffy/Textures/T_Fluffy_Gullet_BC}.RGB
   B=TextureSample(UVs=@n7, Tex=NamedRerouteUsage()) {texture=&T_Fluffy_Gullet_BC}.RGB
   Alpha=@n8.G
  )
  n10 = TextureSample(Tex=NamedRerouteUsage()) {texture=&T_Fluffy_Gullet_BC}
  n11 = TextureSample(Tex=NamedRerouteUsage()) {samplerType=SAMPLERTYPE_LinearColor, texture=&T_Fluffy_Gullet_M}
  NamedRerouteDeclaration(None=Multiply(A=Multiply(A=@n4.RGB) {constB=2}, B=LinearInterpolate(A=@n9, B=@n10.RGB, Alpha=@n11.R))) {name=Color, nodeColor=<rgba 1 0.4 0 1>}
  n12 = TextureSample(UVs=@n6, Tex=NamedRerouteUsage()) {samplerType=SAMPLERTYPE_Normal, texture=/VehicleCosmetics/Wheels/Fluffy/Textures/T_Fluffy_Gullet_N}
  n13 = TextureSample(UVs=@n7, Tex=NamedRerouteUsage()) {samplerType=SAMPLERTYPE_Normal, texture=&T_Fluffy_Gullet_N}
  n14 = NamedRerouteUsage()
  NamedRerouteDeclaration(None=LinearInterpolate(
    A=LinearInterpolate(A=@n12.RGB, B=@n13.RGB, Alpha=@n8.G)
    B=TextureSample(Tex=@n14) {samplerType=SAMPLERTYPE_Normal, texture=&T_Fluffy_Gullet_N}.RGB
    Alpha=@n11.R
   )) {name=Normal, nodeColor=<rgba 0.5 0.5 1 1>}
  n15 = TextureObjectParameter() {group=00 Base, parameterName=Gullet Albedo, sortPriority=0, texture=&T_Fluffy_Gullet_BC}
  NamedRerouteDeclaration(None=@n15) {name=Gullet Albedo, nodeColor=<rgba 1 0.5 0 1>}
  n16 = TextureObjectParameter() {
   group=00 Base
   parameterName=Gullet Masks
   samplerType=SAMPLERTYPE_LinearColor
   sortPriority=1
   texture=&T_Fluffy_Gullet_M
  }
  NamedRerouteDeclaration(None=@n16) {name=Gullet Masks, nodeColor=<rgba 0 1 0 1>}
  n17 = TextureObjectParameter() {
   group=00 Base
   parameterName=Gullet Normal
   samplerType=SAMPLERTYPE_Normal
   sortPriority=2
   texture=&T_Fluffy_Gullet_N
  }
  NamedRerouteDeclaration(None=@n17) {name=Gullet Normal, nodeColor=<rgba 0.5 0.5 1 1>}
  n18 = LinearInterpolate(A=Constant(), B=ScalarParameter() {defaultValue=1, parameterName=Roughness_1}, Alpha=@n8.R)
  n19 = ScalarParameter() {defaultValue=0.4, parameterName=Roughness_2}
  NamedRerouteDeclaration(None=LinearInterpolate(A=LinearInterpolate(A=@n18, B=@n19, Alpha=@n8.G), B=ScalarParameter() {defaultValue=0.2, parameterName=Roughness_3}, Alpha=@n8.B)) {name=Roughness, nodeColor=<rgba 0.6641 1 0 1>}
  NamedRerouteUsage()
 ]
}
